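# Anim Blueprint for the BlondeJaw cosmetic backpack, posing its left and right plate bones to follow the wearer's body.
# A frontend-pose flag, driven by an animation curve, switches between a BodyMesh pose copy and a layered blend adding BodyMesh upper-arm bones to a copied pose.
# Weighted constraints tie each plate to nearby body bones, then partially to neighbouring plates.
# Additive bone-space offsets fine-tune plates and the pack bone.
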AnimBlueprint /BRCosmetics/Accessories/FORT_Backpacks/Backpack_M_MED_BlondeJaw/Meshes/M_MED_BlondeJaw_Pack_AnimBP
target_skeleton=&/M_MED_BlondeJaw_Pack_Skeleton
graphs [
 AnimGraph {nodes [
   "Constraint - dyn_backpackPlate_R_01 (PPPPPPPPPPP)"!(
    ComponentPose
    Alpha=1
    ConstraintWeights_0=0.0178
    ConstraintWeights_1:float=0.0082
    ConstraintWeights_2:float=0.0527
    ConstraintWeights_3:float=0.1003
    ConstraintWeights_4:float=0.0006
    ConstraintWeights_5:float=0.0565
    ConstraintWeights_6:float=0.2539
    ConstraintWeights_7:float=0.2591
    ConstraintWeights_8:float=0.0068
    ConstraintWeights_9:float=0.0987
    ConstraintWeights_10:float=0.1454
   ) -> (Pose) {
    alphaBoolBlend {blendOption=Linear}
    boneToModify {boneName=dyn_backpackPlate_R_01}
    constraintSetup [
     {targetBone {boneName=spine_02}}
     {targetBone {boneName=spine_03}}
     {targetBone {boneName=spine_04}}
     {targetBone {boneName=clavicle_r}}
     {targetBone {boneName=deform_sdk_upperarm_twist_3_r}}
     {targetBone {boneName=deform_sdk_upperarm_twist_2_r}}
     {targetBone {boneName=deform_sdk_upperarm_twist_1_r}}
     {targetBone {boneName=back_r}}
     {targetBone {boneName=spike_01_r}}
     {targetBone {boneName=spike_02_r}}
     {targetBone {boneName=lat_r}}
    ]
    constraintWeights [1,1,1,1,1,1,1,1,1,1,1]
   }
   n1 = Copy Pose From Mesh(SourceMeshComponent=Get BodyMesh() -> (BodyMesh:Skeletal Mesh Component Budgeted Object Reference)) {bCopyCurves=true, bUseAttachedParent=true}
   n2 = Layered blend per bone(
    BasePose=Copy Pose From Mesh() {bCopyCurves=true, bUseAttachedParent=true}
    BlendPoses_0=Copy Pose From Mesh(SourceMeshComponent=Get BodyMesh()) {bCopyCurves=true, bUseAttachedParent=true}
   ) {layerSetup [{branchFilters [
       {boneName=deform_upperarm_bk_base_l}
       {boneName=deform_upperarm_bk_l}
       {boneName=deform_upperarm_up_base_l}
       {boneName=deform_upperarm_up_l}
       {boneName=deform_upperarm_bk_base_r}
       {boneName=deform_upperarm_bk_r}
       {boneName=deform_upperarm_up_base_r}
       {boneName=deform_upperarm_up_r}
      ]}]}
   n3 = Get Frontend_Backpack_Pose() -> (Frontend_Backpack_Pose:bool)
   n4 = "Transform (Modify) Bone - Bone: dyn_backpackPlate_R_02"(
    ComponentPose=Local To Component(LocalPose=Blend Poses by bool(BlendPose_0=@n1, BlendPose_1=@n2, bActiveValue=@n3) {blendType=Linear})
    Alpha=1
    Translation="0,0,1"
    Rotation="0,0,0"
    Scale="1,1,1"
   ) {
    rotationMode=BMM_Additive
    rotationSpace=BCS_BoneSpace
    translationMode=BMM_Additive
    translationSpace=BCS_BoneSpace
    alphaBoolBlend {blendOption=Linear}
    boneToModify {boneName=dyn_backpackPlate_R_02}
    translation {z=1}
   }
   n5 = "Constraint - dyn_backpackPlate_R_01 (PPPPPPPPPPP)"(
    ComponentPose=@n4
    Alpha=0.4
    ConstraintWeights_0=0.0042
    ConstraintWeights_1=0.006
    ConstraintWeights_2=0.0835
    ConstraintWeights_3=0.0013
    ConstraintWeights_4=0.1116
    ConstraintWeights_5=0.0485
    ConstraintWeights_6=0.2497
    ConstraintWeights_7=0.3069
    ConstraintWeights_8=0.0098
    ConstraintWeights_9=0.0513
    ConstraintWeights_10=0.1271
   ) {
    alphaBoolBlend {blendOption=Linear}
    boneToModify {boneName=dyn_backpackPlate_R_01}
    constraintSetup [
     {targetBone {boneName=spine_02}}
     {targetBone {boneName=spine_03}}
     {targetBone {boneName=spine_04}}
     {targetBone {boneName=neck_01}}
     {targetBone {boneName=clavicle_r}}
     {targetBone {boneName=deform_sdk_upperarm_twist_2_r}}
     {targetBone {boneName=deform_sdk_upperarm_twist_1_r}}
     {targetBone {boneName=back_r}}
     {targetBone {boneName=spike_01_r}}
     {targetBone {boneName=spike_02_r}}
     {targetBone {boneName=lat_r}}
    ]
    constraintWeights [1,1,1,1,1,1,1,1,1,1,1]
   }
   n6 = "Constraint - dyn_backpackPlate_R_02 (PPPPPPPPPPPPPPP)"!(
    ComponentPose=@n5
    Alpha=1
    ConstraintWeights_0=0.0843
    ConstraintWeights_1:float=0.2338
    ConstraintWeights_2:float=0.0902
    ConstraintWeights_3:float=0.0761
    ConstraintWeights_4:float=0.0008
    ConstraintWeights_5:float=0.0274
    ConstraintWeights_6:float=0.0058
    ConstraintWeights_7:float=0.0423
    ConstraintWeights_8:float=0.0098
    ConstraintWeights_9:float=0.0957
    ConstraintWeights_10:float=0.0012
    ConstraintWeights_11:float=0.0253
    ConstraintWeights_12:float=0.0599
    ConstraintWeights_13:float=0.0192
    ConstraintWeights_14:float=0.2281
   ) -> (Pose) {
    alphaBoolBlend {blendOption=Linear}
    boneToModify {boneName=dyn_backpackPlate_R_02}
    constraintSetup [
     {targetBone {boneName=spine_01}}
     {targetBone {boneName=spine_02}}
     {targetBone {boneName=spine_03}}
     {targetBone {boneName=spine_04}}
     {targetBone {boneName=clavicle_l}}
     {targetBone {boneName=clavicle_r}}
     {targetBone {boneName=deform_sdk_upperarm_twist_2_r}}
     {targetBone {boneName=deform_sdk_upperarm_twist_1_r}}
     {targetBone {boneName=chest_l}}
     {targetBone {boneName=back_r}}
     {targetBone {boneName=back_l}}
     {targetBone {boneName=spike_02_r}}
     {targetBone {boneName=spike_03_r}}
     {targetBone {boneName=deform_pelvis}}
     {targetBone {boneName=lat_r}}
    ]
    constraintWeights [1,1,1,1,1,1,1,1,1,1,1,1,1,1,1]
   }
   n7 = "Constraint - dyn_backpackPlate_R_03 (PPPPPPPPPPPP)"!(
    ComponentPose=@n6
    Alpha=1
    ConstraintWeights_0=0.2697
    ConstraintWeights_1:float=0.2956
    ConstraintWeights_2:float=0.0546
    ConstraintWeights_3:float=0.0454
    ConstraintWeights_4:float=0.0047
    ConstraintWeights_5:float=0.0084
    ConstraintWeights_6:float=0.0057
    ConstraintWeights_7:float=0.0002
    ConstraintWeights_8:float=0.0054
    ConstraintWeights_9:float=0.0466
    ConstraintWeights_10:float=0.132
    ConstraintWeights_11:float=0.1315
   ) -> (Pose) {
    alphaBoolBlend {blendOption=Linear}
    boneToModify {boneName=dyn_backpackPlate_R_03}
    constraintSetup [
     {targetBone {boneName=spine_01}}
     {targetBone {boneName=spine_02}}
     {targetBone {boneName=spine_03}}
     {targetBone {boneName=spine_04}}
     {targetBone {boneName=clavicle_r}}
     {targetBone {boneName=back_r}}
     {targetBone {boneName=tail1}}
     {targetBone {boneName=tail2}}
     {targetBone {boneName=spike_03}}
     {targetBone {boneName=spike_03_r}}
     {targetBone {boneName=deform_pelvis}}
     {targetBone {boneName=lat_r}}
    ]
    constraintWeights [1,1,1,1,1,1,1,1,1,1,1,1]
   }
   n8 = "Constraint - dyn_backpackPlate_R_04 (PPPPPPPPPPP)"!(
    ComponentPose=@n7
    Alpha=1
    ConstraintWeights_0=0.3078
    ConstraintWeights_1:float=0.1189
    ConstraintWeights_2:float=0.0299
    ConstraintWeights_3:float=0.0311
    ConstraintWeights_4:float=0.0013
    ConstraintWeights_5:float=0.0336
    ConstraintWeights_6:float=0.0031
    ConstraintWeights_7:float=0.0042
    ConstraintWeights_8:float=0.0099
    ConstraintWeights_9:float=0.4519
    ConstraintWeights_10:float=0.0084
   ) -> (Pose) {
    alphaBoolBlend {blendOption=Linear}
    boneToModify {boneName=dyn_backpackPlate_R_04}
    constraintSetup [
     {targetBone {boneName=spine_01}}
     {targetBone {boneName=spine_02}}
     {targetBone {boneName=spine_03}}
     {targetBone {boneName=spine_04}}
     {targetBone {boneName=thigh_twist_01_r}}
     {targetBone {boneName=tail1}}
     {targetBone {boneName=tail2}}
     {targetBone {boneName=spike_03}}
     {targetBone {boneName=spike_03_r}}
     {targetBone {boneName=deform_pelvis}}
     {targetBone {boneName=lat_r}}
    ]
    constraintWeights [1,1,1,1,1,1,1,1,1,1,1]
   }
   n9 = "Constraint - dyn_backpackPlate_L_01 (PPPPPPPPPPPP)"!(
    ComponentPose=@n8
    Alpha=0.5
    ConstraintWeights_0=0.0018
    ConstraintWeights_1:float=0.049
    ConstraintWeights_2:float=0.016
    ConstraintWeights_3:float=0.0766
    ConstraintWeights_4:float=0.0828
    ConstraintWeights_5:float=0.0385
    ConstraintWeights_6:float=0.1861
    ConstraintWeights_7:float=0.2534
    ConstraintWeights_8:float=0.1136
    ConstraintWeights_9:float=0.0051
    ConstraintWeights_10:float=0.0019
    ConstraintWeights_11:float=0.1752
   ) -> (Pose) {
    alphaBoolBlend {blendOption=Linear}
    boneToModify {boneName=dyn_backpackPlate_L_01}
    constraintSetup [
     {targetBone {boneName=spine_01}}
     {targetBone {boneName=spine_02}}
     {targetBone {boneName=spine_03}}
     {targetBone {boneName=spine_04}}
     {targetBone {boneName=clavicle_l}}
     {targetBone {boneName=deform_sdk_upperarm_twist_2_l}}
     {targetBone {boneName=deform_sdk_upperarm_twist_1_l}}
     {targetBone {boneName=back_l}}
     {targetBone {boneName=spike_02_l}}
     {targetBone {boneName=spike_01_l}}
     {targetBone {boneName=spike_03_l}}
     {targetBone {boneName=lat_l}}
    ]
    constraintWeights [1,1,1,1,1,1,1,1,1,1,1,1]
   }
   n10 = "Constraint - dyn_backpackPlate_L_02 (PPPPPPPPPPP)"!(
    ComponentPose=@n9
    Alpha=1
    ConstraintWeights_0=0.0978
    ConstraintWeights_1:float=0.2425
    ConstraintWeights_2:float=0.0819
    ConstraintWeights_3:float=0.0853
    ConstraintWeights_4:float=0.0214
    ConstraintWeights_5:float=0.0257
    ConstraintWeights_6:float=0.0814
    ConstraintWeights_7:float=0.0208
    ConstraintWeights_8:float=0.0805
    ConstraintWeights_9:float=0.0295
    ConstraintWeights_10:float=0.2332
   ) -> (Pose) {
    alphaBoolBlend {blendOption=Linear}
    boneToModify {boneName=dyn_backpackPlate_L_02}
    constraintSetup [
     {targetBone {boneName=spine_01}}
     {targetBone {boneName=spine_02}}
     {targetBone {boneName=spine_03}}
     {targetBone {boneName=spine_04}}
     {targetBone {boneName=clavicle_l}}
     {targetBone {boneName=deform_sdk_upperarm_twist_1_l}}
     {targetBone {boneName=back_l}}
     {targetBone {boneName=spike_02_l}}
     {targetBone {boneName=spike_03_l}}
     {targetBone {boneName=deform_pelvis}}
     {targetBone {boneName=lat_l}}
    ]
    constraintWeights [1,1,1,1,1,1,1,1,1,1,1]
   }
   n11 = "Constraint - dyn_backpackPlate_L_03 (PPPPPPPPPPPP)"!(
    ComponentPose=@n10
    Alpha=1
    ConstraintWeights_0=0.3066
    ConstraintWeights_1:float=0.2861
    ConstraintWeights_2:float=0.0475
    ConstraintWeights_3:float=0.0421
    ConstraintWeights_4:float=0.0026
    ConstraintWeights_5:float=0.0039
    ConstraintWeights_6:float=0.0077
    ConstraintWeights_7:float=0.0005
    ConstraintWeights_8:float=0.0059
    ConstraintWeights_9:float=0.0311
    ConstraintWeights_10:float=0.1577
    ConstraintWeights_11:float=0.1082
   ) -> (Pose) {
    alphaBoolBlend {blendOption=Linear}
    boneToModify {boneName=dyn_backpackPlate_L_03}
    constraintSetup [
     {targetBone {boneName=spine_01}}
     {targetBone {boneName=spine_02}}
     {targetBone {boneName=spine_03}}
     {targetBone {boneName=spine_04}}
     {targetBone {boneName=clavicle_l}}
     {targetBone {boneName=back_l}}
     {targetBone {boneName=tail1}}
     {targetBone {boneName=tail2}}
     {targetBone {boneName=spike_03}}
     {targetBone {boneName=spike_03_l}}
     {targetBone {boneName=deform_pelvis}}
     {targetBone {boneName=lat_l}}
    ]
    constraintWeights [1,1,1,1,1,1,1,1,1,1,1,1]
   }
   n12 = "Constraint - dyn_backpackPlate_L_04 (PPPPPPPPPPP)"!(
    ComponentPose=@n11
    Alpha=1
    ConstraintWeights_0=0.2885
    ConstraintWeights_1:float=0.1223
    ConstraintWeights_2:float=0.035
    ConstraintWeights_3:float=0.0361
    ConstraintWeights_4:float=0.0011
    ConstraintWeights_5:float=0.0378
    ConstraintWeights_6:float=0.0034
    ConstraintWeights_7:float=0.0075
    ConstraintWeights_8:float=0.0147
    ConstraintWeights_9:float=0.445
    ConstraintWeights_10:float=0.0086
   ) -> (Pose) {
    alphaBoolBlend {blendOption=Linear}
    boneToModify {boneName=dyn_backpackPlate_L_04}
    constraintSetup [
     {targetBone {boneName=spine_01}}
     {targetBone {boneName=spine_02}}
     {targetBone {boneName=spine_03}}
     {targetBone {boneName=spine_04}}
     {targetBone {boneName=thigh_twist_01_l}}
     {targetBone {boneName=tail1}}
     {targetBone {boneName=tail2}}
     {targetBone {boneName=spike_03}}
     {targetBone {boneName=spike_03_l}}
     {targetBone {boneName=deform_pelvis}}
     {targetBone {boneName=lat_l}}
    ]
    constraintWeights [1,1,1,1,1,1,1,1,1,1,1]
   }
   n13 = "Constraint - dyn_backpackPlate_R_01 (P)"(ComponentPose=@n12, Alpha=0.3, ConstraintWeights_0=1) {
    alphaBoolBlend {blendOption=Linear}
    boneToModify {boneName=dyn_backpackPlate_R_01}
    constraintSetup [{targetBone {boneName=dyn_backpackPlate_R_02}}]
    constraintWeights [1]
   }
   n14 = "Constraint - dyn_backpackPlate_R_02 (PP)"!(ComponentPose=@n13, Alpha=0.5, ConstraintWeights_0=0.8, ConstraintWeights_1:float=0.2) -> (Pose) {
    alphaBoolBlend {blendOption=Linear}
    boneToModify {boneName=dyn_backpackPlate_R_02}
    constraintSetup [{targetBone {boneName=dyn_backpackPlate_R_01}}, {targetBone {boneName=dyn_backpackPlate_R_03}}]
    constraintWeights [1,1]
   }
   n15 = "Constraint - dyn_backpackPlate_R_03 (PP)"!(ComponentPose=@n14, Alpha=0.5, ConstraintWeights_0=0.8, ConstraintWeights_1:float=0.2) -> (Pose) {
    alphaBoolBlend {blendOption=Linear}
    boneToModify {boneName=dyn_backpackPlate_R_03}
    constraintSetup [{targetBone {boneName=dyn_backpackPlate_R_02}}, {targetBone {boneName=dyn_backpackPlate_R_04}}]
    constraintWeights [1,1]
   }
   n16 = "Constraint - dyn_backpackPlate_L_01 (P)"(ComponentPose=@n15, Alpha=0.2, ConstraintWeights_0=1) {
    alphaBoolBlend {blendOption=Linear}
    boneToModify {boneName=dyn_backpackPlate_L_01}
    constraintSetup [{targetBone {boneName=dyn_backpackPlate_L_02}}]
    constraintWeights [1]
   }
   n17 = "Constraint - dyn_backpackPlate_L_02 (PP)"!(ComponentPose=@n16, Alpha=0.5, ConstraintWeights_0=0.8, ConstraintWeights_1:float=0.2) -> (Pose) {
    alphaBoolBlend {blendOption=Linear}
    boneToModify {boneName=dyn_backpackPlate_L_02}
    constraintSetup [{targetBone {boneName=dyn_backpackPlate_L_01}}, {targetBone {boneName=dyn_backpackPlate_L_03}}]
    constraintWeights [1,1]
   }
   n18 = "Constraint - dyn_backpackPlate_L_03 (PP)"!(ComponentPose=@n17, Alpha=0.3, ConstraintWeights_0=0.8, ConstraintWeights_1:float=0.2) -> (Pose) {
    alphaBoolBlend {blendOption=Linear}
    boneToModify {boneName=dyn_backpackPlate_L_03}
    constraintSetup [{targetBone {boneName=dyn_backpackPlate_L_02}}, {targetBone {boneName=dyn_backpackPlate_L_04}}]
    constraintWeights [1,1]
   }
   n19 = "Transform (Modify) Bone - Bone: dyn_backpackPlate_R_01"(ComponentPose=@n18, Alpha=1, Translation="0,0.04,1.7", Rotation="-1,0,7", Scale="1,1,1") {
    rotationMode=BMM_Additive
    rotationSpace=BCS_BoneSpace
    translationMode=BMM_Additive
    translationSpace=BCS_BoneSpace
    alphaBoolBlend {blendOption=Linear}
    boneToModify {boneName=dyn_backpackPlate_R_01}
    rotation {pitch=-1, roll=7}
    translation {y=0.04, z=1.7}
   }
   n20 = "Transform (Modify) Bone - Bone: dyn_backpackPlate_R_02"(ComponentPose=@n19, Alpha=1, Translation="0,-1,0", Rotation="-1,0,4", Scale="1,1,1") {
    rotationMode=BMM_Additive
    rotationSpace=BCS_BoneSpace
    translationMode=BMM_Additive
    translationSpace=BCS_BoneSpace
    alphaBoolBlend {blendOption=Linear}
    boneToModify {boneName=dyn_backpackPlate_R_02}
    rotation {pitch=-1, roll=4}
    translation {y=-1}
   }
   n21 = "Transform (Modify) Bone - Bone: dyn_backpackPlate_R_03"(ComponentPose=@n20, Alpha=1, Translation="0,-1,0", Rotation="-1,0,0", Scale="1,1,1") {
    rotationMode=BMM_Additive
    rotationSpace=BCS_BoneSpace
    translationMode=BMM_Additive
    translationSpace=BCS_BoneSpace
    alphaBoolBlend {blendOption=Linear}
    boneToModify {boneName=dyn_backpackPlate_R_03}
    rotation {pitch=-1}
    translation {y=-1}
   }
   n22 = "Transform (Modify) Bone - Bone: dyn_backpackPlate_L_01"(ComponentPose=@n21, Alpha=1, Translation="0,0,1.2", Rotation="0,0,7", Scale="1,1,1") {
    rotationMode=BMM_Additive
    rotationSpace=BCS_BoneSpace
    translationMode=BMM_Additive
    translationSpace=BCS_BoneSpace
    alphaBoolBlend {blendOption=Linear}
    boneToModify {boneName=dyn_backpackPlate_L_01}
    rotation {roll=7}
    translation {z=1.2}
   }
   n23 = "Transform (Modify) Bone - Bone: pack"(ComponentPose=@n22, Translation="0,-2,0", Rotation="0,0,0") {
    rotationMode=BMM_Additive
    rotationSpace=BCS_BoneSpace
    translationMode=BMM_Additive
    translationSpace=BCS_BoneSpace
    alphaBoolBlend {blendOption=Linear}
    boneToModify {boneName=pack}
    translation {y=-2}
   }
   Output Pose(Result=Component To Local(ComponentPose=@n23)) {layerGroup=None}
  ]}
 EventGraph {nodes [
   Event BlueprintUpdateAnimation()
   > n1 = Branch(Condition=Get bIsInFrontEnd())
   n2 = GetCurveValue(CurveName=Frontend_Backpack_Pose)
   > Set Frontend_Backpack_Pose(Frontend_Backpack_Pose:bool="float > float"(A=@n2)) -> (Output_Get:bool)
   n3 = "Absolute (Float)"(A=float - float(A=@n2, B=1))
   > Set in game tweak alpha(in game tweak alpha:double=@n3) -> (Output_Get:double)
   Set in game tweak alpha(execute=@n1.else, in game tweak alpha=@n3)
   n4 = Event OnCharacterCustomizationLoaded()
   > Set BodyMesh(BodyMesh:Skeletal Mesh Component Budgeted Object Reference=Get Array Item with EFort Custom Part Type(Array=@n4.CustomMeshParts)) -> (Output_Get:Skeletal Mesh Component Budgeted Object Reference)
   TryGetPawnOwner()
   "To Float (Boolean)"(InBool=false)
  ]}
]
variables [
 BodyMesh {type=SkeletalMeshComponentBudgeted}
 Frontend_Backpack_Pose {type=boolean}
 in game tweak alpha {type=number}
 OwnerSkelMesh {type=SkeletalMeshComponent}
]
